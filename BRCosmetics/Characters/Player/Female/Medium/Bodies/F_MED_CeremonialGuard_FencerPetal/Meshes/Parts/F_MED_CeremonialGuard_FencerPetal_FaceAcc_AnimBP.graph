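# Post-process-style AnimBlueprint for the FencerPetal female Ceremonial Guard face accessory mesh, adding secondary motion over the parent body's pose.
# Copy Pose From Mesh takes the attached parent's pose into a RigidBody simulation using the FaceAcc physics asset, based at spine_05.
# The simulated pose passes through the FaceAcc Control Rig to the output.
# A chain of four hair-bone AnimDynamics nodes remains unconnected to the output, and the event graph is empty.
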AnimBlueprint /BRCosmetics/Characters/Player/Female/Medium/Bodies/F_MED_CeremonialGuard_FencerPetal/Meshes/Parts/F_MED_CeremonialGuard_FencerPetal_FaceAcc_AnimBP
target_skeleton=/Game/Characters/Player/Female/Medium/Bodies/F_MED_CeremonialGuard_FencerPetal/Meshes/Parts/F_MED_CeremonialGuard_FencerPetal_FaceAcc_Skeleton
graphs [
 AnimGraph {nodes [
   n1 = Anim Dynamics - Bone: dyn_R_hair(Alpha=1, ExternalForce="0,0,0") {
    angularDampingOverride=0.9999
    angularSpringConstant=250
    bAngularSpring=true
    bOverrideAngularDamping=true
    bUseGravityOverride=true
    simulationSpace=RootRelative
    alphaBoolBlend {blendOption=Linear}
    boundBone {boneName=dyn_R_hair}
    componentAppliedLinearAccClamp=<10000 10000 10000>
    componentLinearAccScale=<1 1 1>
    componentLinearVelScale=<1 1 1>
    gravityOverride {y=-200, z=-920}
    physicsBodyDefinitions [{
      boundBone {boneName=dyn_R_hair}
      boxExtents=<2 2 2>
      constraintSetup {angularLimitsMax=<10 30 20>, angularLimitsMin=<-4 -10 -3>, angularTarget=<0 2 0>, linearAxesMax=<0.3 0.3 0.3>}
      localJointOffset=<0.1 0 0>
     }]
    relativeSpaceBone {boneName=head}
   }
   n2 = Anim Dynamics - Bone: dyn_L_hair(ComponentPose=@n1, Alpha=1, ExternalForce="0,0,0") {
    angularDampingOverride=0.9999
    angularSpringConstant=250
    bAngularSpring=true
    bOverrideAngularDamping=true
    bUseGravityOverride=true
    simulationSpace=RootRelative
    alphaBoolBlend {blendOption=Linear}
    boundBone {boneName=dyn_L_hair}
    componentAppliedLinearAccClamp=<10000 10000 10000>
    componentLinearAccScale=<1 1 1>
    componentLinearVelScale=<1 1 1>
    gravityOverride {z=-920}
    physicsBodyDefinitions [{
      boundBone {boneName=dyn_L_hair}
      boxExtents=<3 3 3>
      constraintSetup {angularLimitsMax=<15 30 20>, angularLimitsMin=<0 0 -3>, angularTarget=<0 2 0>, linearAxesMax=<0.3 0.3 0.3>}
      localJointOffset=<0.1 0 0>
     }]
    relativeSpaceBone {boneName=head}
   }
   n3 = Anim Dynamics - Bone: dyn_hair_p(ComponentPose=@n2, Alpha=1, ExternalForce="0,0,0") {
    angularDampingOverride=0.99
    angularSpringConstant=250
    bAngularSpring=true
    bUseGravityOverride=true
    linearSpringConstant=500
    simulationSpace=BoneRelative
    alphaBoolBlend {blendOption=Linear}
    boundBone {boneName=dyn_hair_p}
    componentAppliedLinearAccClamp=<5000 5000 5000>
    componentLinearAccScale=<1 1 1>
    componentLinearVelScale=<1 1 1>
    gravityOverride {z=-980}
    physicsBodyDefinitions [{
      boundBone {boneName=dyn_hair_p}
      boxExtents=<5 7 7>
      constraintSetup {angularLimitsMax=<3 10 10>, angularLimitsMin=<-3 -5 -10>, angularTarget=<0 0 0>}
      localJointOffset=<0.15 0 0>
     }]
    relativeSpaceBone {boneName=spine_05}
   }
   Anim Dynamics - Bone: pasted__joint3(ComponentPose=@n3, Alpha=1, ExternalForce="0,0,0") {
    angularDampingOverride=0.9999
    angularSpringConstant=250
    bAngularSpring=true
    bOverrideAngularDamping=true
    bUseGravityOverride=true
    simulationSpace=RootRelative
    alphaBoolBlend {blendOption=Linear}
    boundBone {boneName=pasted__joint3}
    componentAppliedLinearAccClamp=<10000 10000 10000>
    componentLinearAccScale=<1 1 1>
    componentLinearVelScale=<1 1 1>
    gravityOverride {z=-920}
    physicsBodyDefinitions [{
      boundBone {boneName=pasted__joint3}
      boxExtents=<5 5 5>
      constraintSetup {angularLimitsMax=<15 30 20>, angularLimitsMin=<0 0 -3>, angularTarget=<0 2 0>, linearAxesMax=<0.3 0.3 0.3>}
      localJointOffset=<0.1 0 0>
     }]
    relativeSpaceBone {boneName=head}
   }
   n4 = RigidBody!(
    ComponentPose=Local To Component(LocalPose=Copy Pose From Mesh() {bUseAttachedParent=true})
    Alpha
    OverrideWorldGravity:Vector="0,0,-900"
    ExternalForce
   ) -> (Pose) {
    bOverrideWorldGravity=true
    overridePhysicsAsset=/Game/Characters/Player/Female/Medium/Bodies/F_MED_CeremonialGuard_FencerPetal/Meshes/Parts/F_MED_CeremonialGuard_FencerPetal_FaceAcc_Physics
    simulationSpace=BaseBoneSpace
    alphaBoolBlend {blendOption=Linear}
    baseBoneRef {boneName=spine_05}
    componentAppliedLinearAccClamp=<7000 7000 7000>
    componentLinearAccScale {z=1}
    componentLinearVelScale {z=1}
    overrideWorldGravity {z=-400}
    simSpaceSettings {
     maxAngularAcceleration=200
     maxAngularVelocity=3
     maxLinearAcceleration=200
     maxLinearVelocity=150
     velocityScaleZ=0
     worldAlpha=1
     externalLinearDragV=<0.1 0.1>
    }
   }
   n5 = Control Rig(Source=Component To Local(ComponentPose=@n4)) {
    alphaBoolBlend {blendOption=Linear}
    controlRigAssetReference {blueprintRigClass=/Game/Characters/Player/Female/Medium/Bodies/F_MED_CeremonialGuard_FencerPetal/Meshes/Parts/F_MED_CeremonialGuard_FencerPetal_FaceAcc_CtrlRig.F_MED_CeremonialGuard_FencerPetal_FaceAcc_CtrlRig_C}
   }
   Output Pose(Result=@n5) {layerGroup=None}
  ]}
 EventGraph {nodes [Event BlueprintUpdateAnimation()]}
]
variables []
